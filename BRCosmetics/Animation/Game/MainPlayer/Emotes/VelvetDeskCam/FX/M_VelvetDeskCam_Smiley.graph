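# Additive lit material for the VelvetDeskCam emote's smiley smoke effect, usable on Niagara sprites and meshes.
# Panned, noise-distorted smoke is shaped by erosion, radial and inverted-Fresnel masks.
# Separate channels of a smiley erosion texture form eye and mouth masks, each revealed through EyesRGMouthB.
# Particle color and dynamic parameters supply tint and animation controls, while a sine wave plus side deforms build a switchable world position offset.
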Material /BRCosmetics/Animation/Game/MainPlayer/Emotes/VelvetDeskCam/FX/M_VelvetDeskCam_Smiley
graph {
 properties {
  bAutomaticallySetUsageInEditor=false
  bUsedWithMeshDeformer=true
  bUsedWithNiagaraMeshParticles=true
  bUsedWithNiagaraSprites=true
  bUsedWithSkeletalMesh=true
  blendMode=BLEND_Additive
  previewMesh=&/SM_VelvetDeskCam_Smiley_FX_New
  translucencyLightingMode=TLM_SurfacePerPixelLighting
  usedShadingModels=Default Lit
 }
 nodes [
  n1 = TextureCoordinate()
  n2 = ScalarParameter() {defaultValue=3, parameterName=WPO_SineTiling}
  n3 = Time()
  n4 = ComponentMask(None=Add(A=Multiply(A=@n1, B=@n2) {constB=3}, B=Multiply(A=@n3) {constB=0.1}) {constB=-0.5}) {g=true}
  n5 = ScalarParameter() {defaultValue=2, parameterName=SineValue_Min}
  n6 = ScalarParameter() {defaultValue=-2, parameterName=SineValue_Max}
  n7 = MaterialFunctionCall(Radius=Constant() {r=0.5}, Density=Constant() {r=2}) {materialFunction=&RadialGradientExponential}
  n8 = Constant()
  n9 = MaterialFunctionCall(
   X=Multiply(
    A=MaterialFunctionCall!(Sine Phase=@n4, Value 1=@n5, Value 2=@n6) {materialFunction=/Engine/Functions/Engine_MaterialFunctions02/Utility/Sine_Remapped}.Remapped Sine
    B=SmoothStep(Value=@n7.RadialGradientExponential) {constMax=0, constMin=1}
   )
   Y=@n8
   Z=@n8
  ) {materialFunction=&MakeFloat3}
  n10 = TextureCoordinate()
  n11 = Clamp(None=ComponentMask(None=Add(A=@n10) {constB=-0.35}) {r=true})
  n12 = Constant()
  n13 = Add(A=@n10) {constB=0.5}
  n14 = Clamp(None=OneMinus(None=ComponentMask(None=@n13) {r=true}))
  n15 = Constant()
  n16 = Add(
   A=Multiply(
    A=MaterialFunctionCall(X=@n11, Y=@n12, Z=@n12) {materialFunction=&MakeFloat3}.Result
    B=ScalarParameter() {defaultValue=40, parameterName=Side1Deform}
   ) {constB=-50}
   B=Multiply(
    A=MaterialFunctionCall(X=@n14, Y=@n15, Z=@n15) {materialFunction=&MakeFloat3}.Result
    B=ScalarParameter() {parameterName=Side2Deform}
   ) {constB=50}
  )
  LinearInterpolate(
   A=Add(A=Transform(None=@n9.Result) {transformSourceType=TRANSFORMSOURCE_Local}, B=Transform(None=@n16) {transformSourceType=TRANSFORMSOURCE_Local})
   Alpha=NamedRerouteUsage()
  ) {constAlpha=0, constB=0}
  n17 = NamedRerouteUsage()
  n18 = Constant() {r=0.4}
  n19 = NamedRerouteUsage()
  n20 = Multiply(A=TextureSample() {texture=/Game/Athena/Helios/Maps/Test/Dev/Yi/MaterialAndMesh/SmokeMask_02_T}.R, B=NamedRerouteUsage())
  n21 = NamedRerouteUsage()
  n22 = OneMinus(None=Fresnel(ExponentIn=LinearInterpolate(B=@n20, Alpha=@n21) {constA=1}) {exponent=1})
  n23 = Power(Base=@n22)
  n24 = Power(Base=@n22) {constExponent=1.5}
  n25 = Abs(None=DotProduct(A=CameraVectorWS(), B=Constant3Vector() {constant=<rgba 1 0 0 0>}))
  n26 = Saturate(None=OneMinus(None=Power(Base=@n25) {constExponent=4}))
  n27 = Saturate(None=LinearInterpolate(A=@n23, B=LinearInterpolate(A=@n24, B=@n23, Alpha=@n26), Alpha=ScalarParameter() {defaultValue=1, parameterName=FresnelViewAngle}))
  n28 = NamedRerouteUsage()
  n29 = Subtract(A=Subtract(A=Subtract(A=@n27, B=@n28), B=NamedRerouteUsage()), B=NamedRerouteUsage())
  n30 = NamedRerouteUsage()
  n31 = Multiply(
   A=Multiply(
    A=MaterialFunctionCall!("Erosion Texture (RGBA)"=@n17, Erosion Contrast=@n18, Erosion Amount=@n19) {materialFunction=/Game/Effects/Niagara/MasterMaterials/Functions/Mojo_MF/MF_Erosion_Complex_Base}.Color
    B=Clamp(None=@n29)
   )
   B=@n30
  )
  n32 = NamedRerouteUsage()
  Multiply(A=Multiply(A=Multiply(A=@n31, B=@n32) {constB=2}, B=ScalarParameter() {defaultValue=1, parameterName=GlobalOpacity}), B=NamedRerouteUsage())
  Multiply(A=@n30, B=NamedRerouteUsage())
  n33 = DynamicParameter() {parameterIndex=3, defaultValue=<rgba -1 1 0.025 1>, paramNames [Erosion, UVSwipe, FaceUVDistortion, NonTriplanarScale]}
  NamedRerouteDeclaration(None=@n33.Erosion) {name=Erosion, nodeColor=<rgba 0.875 0 1 1>}
  n34 = DynamicParameter() {parameterIndex=1, defaultValue=<rgba 0 1 0 0>, paramNames [MistMask,MistOpacity,Param3,Param4]}
  NamedRerouteDeclaration(None=@n34.MistMask) {name=MistMask, nodeColor=<rgba 0 0.6719 1 1>}
  n35 = DynamicParameter() {defaultValue=<rgba 1 0.3 1 1>, paramNames [WPOEnabled,CenterMask,SmokeOpacity,Param4]}
  NamedRerouteDeclaration(None=@n35.WPOEnabled) {name=WPOEnabled, nodeColor=<rgba 0.6875 0 1 1>}
  NamedRerouteDeclaration(None=@n33.NonTriplanarScale) {name=NonTriplanarScale, nodeColor=<rgba 0 1 0.3672 1>}
  NamedRerouteDeclaration(None=@n34.MistOpacity) {name=MistOpacity, nodeColor=<rgba 0 0.2734 1 1>}
  NamedRerouteDeclaration(None=@n35.CenterMask) {name=CenterMask, nodeColor=<rgba 0 0.5312 1 1>}
  NamedRerouteDeclaration(None=@n35.SmokeOpacity) {name=SmokeOpacity, nodeColor=<rgba 1 0 0.9375 1>}
  n36 = ParticleColor()
  NamedRerouteDeclaration(None=@n36.A) {name=ParticleAlpha, nodeColor=<rgba 1 0 0.3281 1>}
  n37 = NamedRerouteUsage()
  n38 = Constant() {r=1}
  n39 = Multiply(A=Multiply(A=TextureCoordinate() {uTiling=1.5, vTiling=1.5}, B=AppendVector(A=@n37, B=@n38)), B=NamedRerouteUsage())
  n40 = Multiply(A=TextureSample(UVs=Panner() {speedX=0.03, speedY=-0.05}) {samplerType=SAMPLERTYPE_Masks, texture=/Game/Weapons/WeaponSkins/Wraps/Textures/Mash/Mash_Noise}.R) {constB=0.05}
  n41 = Panner(Coordinate=Add(A=Add(A=@n39, B=@n40)) {constB=0.3}) {speedY=0.05}
  n42 = Saturate(None=Add(A=TextureSample(UVs=@n41) {texture=&/T_DM_Soft_Smoke_Tile}.R) {constB=0.1})
  n43 = ScalarParameter() {defaultValue=0.15, parameterName=Add1}
  n44 = Clamp(None=Add(A=Add(A=@n42, B=@n43) {constB=0.15}) {constB=0.15})
  NamedRerouteDeclaration(None=@n44) {name=SmokeLayerPre, nodeColor=<rgba 1 0 0.1172 1>}
  n45 = SmoothStep(Value=MaterialFunctionCall(Radius=ScalarParameter() {defaultValue=0.5, parameterName=Radius}, Density=ScalarParameter() {defaultValue=3, parameterName=Density}) {materialFunction=&RadialGradientExponential}.RadialGradientExponential) {constMax=0, constMin=1}
  n46 = ScalarParameter() {parameterName=InvertSideMask}
  n47 = MaterialFunctionCall(
   "Erosion Texture (RGBA)"=@n42
   Erosion Contrast=Constant() {r=0.1}
   Erosion Amount=LinearInterpolate(A=@n45, B=OneMinus(None=@n45), Alpha=@n46)
  ) {materialFunction=&MF_Erosion_Complex_Base}
  n48 = SmoothStep(Value=MaterialFunctionCall(Radius=ScalarParameter() {defaultValue=0.5, parameterName=Radius2}, Density=ScalarParameter() {defaultValue=2, parameterName=Density3}) {materialFunction=&RadialGradientExponential}.RadialGradientExponential)
  n49 = LinearInterpolate(A=@n48, B=Power(Base=OneMinus(None=@n48)) {constExponent=10}, Alpha=@n46)
  NamedRerouteDeclaration(None=LinearInterpolate(B=LinearInterpolate(A=@n47.Color, Alpha=@n49), Alpha=ScalarParameter() {parameterName=EnableSmileyEdgeMask}) {constA=1}) {name=SmileyEdgeMask, nodeColor=<rgba 1 0.375 0 1>}
  NamedRerouteDeclaration(None=@n33.UVSwipe) {name=UVSwipe, nodeColor=<rgba 0 0.8594 1 1>}
  NamedRerouteDeclaration(None=@n36.RGB) {name=ParticleRGB, nodeColor=<rgba 0 1 0.3906 1>}
  NamedRerouteDeclaration(None=@n33.FaceUVDistortion) {name=FaceUVDistortion, nodeColor=<rgba 1 0.7031 0 1>}
  n50 = NamedRerouteUsage()
  n51 = SmoothStep(Value=MaterialFunctionCall(UVs=TextureCoordinate(), Radius=NamedRerouteUsage(), Density=Constant() {r=2.5}) {materialFunction=&RadialGradientExponential}.RadialGradientExponential) {constMax=0, constMin=1}
  n52 = ScalarParameter() {parameterName=CenterMaskSmoke}
  NamedRerouteDeclaration(None=Multiply(A=Multiply(A=@n44, B=@n50), B=LinearInterpolate(B=@n51, Alpha=@n52) {constA=1})) {name=SmokeLayerMixed, nodeColor=<rgba 0 1 0.625 1>}
  n53 = TextureCoordinate()
  n54 = ScalarParameter() {defaultValue=-0.25, parameterName=SmileyX}
  n55 = ScalarParameter() {defaultValue=-0.25, parameterName=SmileyY}
  n56 = Panner(Coordinate=TextureCoordinate() {uTiling=0.7, vTiling=0.7}) {speedX=0.02, speedY=0.02}
  n57 = Add(
   A=Add(A=Multiply(A=@n53) {constB=1.5}, B=AppendVector(A=@n54, B=@n55))
   B=Multiply(
    A=TextureSample(UVs=@n56) {samplerType=SAMPLERTYPE_Masks, texture=&Mash_Noise}.R
    B=NamedRerouteUsage()
   ) {constB=0.025}
  )
  n58 = VectorParameter() {parameterName=EyesRGMouthB, defaultValue=<rgba 0 0 1 1>}
  n59 = Multiply(A=Add(A=ComponentMask(None=@n53) {r=true}, B=LinearInterpolate(Alpha=@n58.B) {constA=0.3, constB=-0.8}) {constB=0}) {constB=3}
  n60 = Multiply(
   A=TextureSample(UVs=@n57) {samplerSource=SSM_Clamp_WorldGroupSettings, texture=&/T_SmileyErosion3}.B
   B=OneMinus(None=Clamp(None=@n59))
  )
  NamedRerouteDeclaration(None=Multiply(A=Clamp(None=@n60)) {constB=10}) {name=MaskB, nodeColor=<rgba 0.2891 0 1 1>}
  n61 = ScalarParameter() {defaultValue=0.5, parameterName=Eye2XMult}
  n62 = ScalarParameter() {defaultValue=1, parameterName=Eye1YMult}
  n63 = Add(
   A=Multiply(A=@n57, B=AppendVector(A=@n61, B=@n62))
   B=AppendVector(A=ScalarParameter() {defaultValue=0.25, parameterName=Eye2X}, B=ScalarParameter() {parameterName=Eye1Y})
  )
  NamedRerouteDeclaration(None=LinearInterpolate(
    A=TextureSample(UVs=@n63) {samplerSource=SSM_Clamp_WorldGroupSettings, texture=&T_SmileyErosion3}.G
    Alpha=LinearInterpolate(Alpha=@n58.G) {constA=1, constB=-2}
   ) {constAlpha=1, constB=0}) {name=MaskG, nodeColor=<rgba 0.0312 1 0 1>}
  n64 = ScalarParameter() {defaultValue=0.5, parameterName=Eye1XMult}
  n65 = ScalarParameter() {defaultValue=1, parameterName=Eye1YMult}
  n66 = Add(
   A=Multiply(A=@n57, B=AppendVector(A=@n64, B=@n65))
   B=AppendVector(A=ScalarParameter() {defaultValue=0.2, parameterName=Eye1X}, B=ScalarParameter() {parameterName=Eye1Y})
  )
  NamedRerouteDeclaration(None=LinearInterpolate(
    A=TextureSample(UVs=@n66) {samplerSource=SSM_Clamp_WorldGroupSettings, texture=&T_SmileyErosion3}.R
    Alpha=LinearInterpolate(Alpha=@n58.R) {constA=1, constB=-2}
   ) {constAlpha=1, constB=0}) {name=MaskR, nodeColor=<rgba 1 0 0.3516 1>}
 ]
}
