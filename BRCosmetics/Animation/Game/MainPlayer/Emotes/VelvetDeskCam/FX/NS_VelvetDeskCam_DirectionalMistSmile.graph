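# Niagara effect for the VelvetDeskCam emote that releases a short, faint puff of mist from a single sprite emitter named Smoke.
# Spawn rate and burst modules feed a spawn point that sweeps along X over emitter age, laying a short trail pulled toward the camera.
# Sprites get a small initial push, then grow, slow under drag and fade while animating SubUV frames from random starts.
# Fortnite time-of-day inputs tint and dim the color.
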NiagaraSystem /BRCosmetics/Animation/Game/MainPlayer/Emotes/VelvetDeskCam/FX/NS_VelvetDeskCam_DirectionalMistSmile
dependencies {
 usedDataInterfaces [
  &NiagaraDataInterfaceCamera
  &NiagaraDataInterfaceCurve
  &NiagaraDataInterfaceSpriteRendererInfo
  &NiagaraDataInterfaceColorCurve
 ]
 usedDynamicInputs [
  &Lerp_Float
  &Max_Float
  &Multiply_Float
  &DotProduct
  &NormalizeVector
  &ReturnCameraVector
 ]
 usedModules [&SystemState]
}
emitters {Smoke {
  emitterUpdateScript [
   &EmitterState(Life Cycle Mode=::NewEnumerator1, Loop Behavior=::NewEnumerator1, Loop Duration=0.4, UseLoopDelay=false)
   &SpawnRate(SpawnRate=10, Use Spawn Probability=false)
   &SpawnBurst_Instantaneous(Use Spawn Probability=false, UseLoopCountLimit=false)
  ]
  particleSpawnScript [
   &InitializeParticle(
    Color=&Multiply_LinearColorByFloat(
     Linear Color=&FNTimeOfDayColor(Night Color=<rgba 0.4977 0.6355 1 1>, Morning Color=<rgba 1 0.5989 0.7062 1>, Day Color=<rgba 1 1 1 1>, Evening Color=<rgba 1 1 1 1>)
     Float=&FNTimeOfDayScalar(Night Scale=0.5, Morning Scale=1, Evening Scale=1)
    )
    Lifetime=1.4
    Sprite Rotation Mode=::NewEnumerator2
    Sprite Size Mode=::NewEnumerator3
    Sprite UV Mode=::NewEnumerator4
    Uniform Sprite Size=25
    UsePositionOffset=false
    Write Material Random=true
   )
   &SystemLocation(
    MaskSpawnGroup=false
    Offset=&Add_Vector(
     A=&Multiply_VectorByFloat(
      Vector=<20 0 0>
      Float=&FloatFromCurve(
       FloatCurve=DataInterface() {propertyValues {Curve {keys [{tangentMode=RCTM_User, time=0, value=0}, {tangentMode=RCTM_User, time=0.263, value=1}]}}}
       CurveIndex=$Emitter.Age
      )
     )
     B=<0 0 -5>
    )
   )
   &PointForce(
    Force Strength=&RandomRangeFloat(Minimum=1, Maximum=1)
    Linear Force Vector:Vector3f=<0 1 0>
    Module.Use Falloff=false
    Module.UseLinearForce=true
   )
   &CameraOffset(Camera Offset Amount=30)
   &DynamicMaterialParameters(
    Index 0 Param 3=1
    Module.Index0Float0Write=true
    Module.Index0Float1Write=true
    Module.Index0Float2Write=true
    Module.Index0Float3Write=true
   )
   &ApplyInitialForces()
  ]
  particleUpdateScript [
   &ParticleState()
   &ScaleSpriteSize(Uniform Curve Sprite Scale=DataInterface() {propertyValues {Curve {keys [{tangentMode=RCTM_User, time=0, value=0.7609}, {tangentMode=RCTM_User, time=1, value=1.2098}]}}})
   &SubUVAnimation(
    End Frame Range Override=30
    "Sprite Renderer "=DataInterface() {propertyValues {SpriteRenderer=&self:Smoke_0.NiagaraSpriteRendererProperties_2}}
    Start Frame Range Override=&RandomRangeInt(Minimum=0, Maximum=5)
    SubUV Lookup Index Scale=0.4
    UseEndFrame=false
    UseStartFrame=true
   )
   &Drag(
    Drag=&FloatFromCurve(FloatCurve=DataInterface() {propertyValues {Curve {keys [{interpMode=RCIM_Cubic, tangentMode=RCTM_User, time=0, value=0}, {interpMode=RCIM_Cubic, tangentMode=RCTM_User, time=0.6258, value=1}]}}})
    Use Linear Drag=true
    Use Rotational Drag=true
   )
   &SolveForcesAndVelocity(Clamp Velocity=false, Limit Acceleration=false)
   &ScaleColor(
    Scale Alpha=&FloatFromCurve(
     FloatCurve=DataInterface() {propertyValues {Curve {keys [
         {arriveTangent=4.7901, interpMode=RCIM_Cubic, leaveTangent=4.7901, tangentMode=RCTM_User, time=0, value=0}
         {arriveTangent=2, interpMode=RCIM_Cubic, leaveTangent=2, time=0.0628, value=0.7605}
         {interpMode=RCIM_Cubic, tangentMode=RCTM_User, time=0.5, value=1}
         {tangentMode=RCTM_User, time=1, value=0}
        ]}}}
     Scale Curve=0.2
    )
    ScaleA=true
    ScaleRGB=true
   ) {module=ScaleColor001}
  ]
  rendererClasses [&NiagaraSpriteRendererProperties]
  renderers [{rendererClass=&NiagaraSpriteRendererProperties}]
 }}
summary {
 systemName=NS_VelvetDeskCam_DirectionalMistSmile
 emitters [{emitterName=Smoke, rendererClasses [&NiagaraSpriteRendererProperties]}]
 userVariables []
}
