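# Niagara mist spray effect for the VelvetDeskCam emote, built from soft smoke and frost sprites.
# Smoke spawns by burst, rate and distance travelled, fires in a cone, then grows, slows and sinks under curve-driven material parameters.
# FrostSprites shoots flipbook puffs forward with heavy drag and time-of-day tinting.
# Both partly inherit source movement; Cone and Spray_Jet show no modules or renderers.
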NiagaraSystem /BRCosmetics/Animation/Game/MainPlayer/Emotes/VelvetDeskCam/FX/NS_VelvetDeskCam_MistSpray
dependencies {
 usedDataInterfaces [&NiagaraDataInterfaceCurve, &NiagaraDataInterfaceVector2DCurve]
 usedDynamicInputs [&Multiply_Vector2D_ByFloat]
 usedModules [&SystemState]
}
emitters {
 Cone {rendererClasses [], renderers []}
 FN_Core_Smoke_Sprite {
  emitterUpdateScript [
   &EmitterState()
   &SpawnBurst_Instantaneous(Spawn Count=2, Use Spawn Probability=false)
   &SpawnRate(SpawnRate=10, Use Spawn Probability=false)
   &SpawnPerUnit(
    Module.Use Max Movement Threshold=true
    Module.Use Movement Tolerance=false
    Spawn Probability=1
    Spawn Spacing=5
    Use Spawn Probability=true
   )
  ]
  particleSpawnScript [
   /SetVariables(Particles.InitSpriteRotRate=&RandomRangeFloat(Minimum=-2, Maximum=2)) {isSetParameters=true}
   &InitializeParticle(
    Color=<rgba 1 0.9588 0.9684 0.02>
    Lifetime=&Multiply_Float(A=&RandomRangeFloat(Minimum=1, Maximum=2.6), B=1)
    Sprite Rotation Angle=&RandomRangeFloat(Minimum=0, Maximum=360)
    Sprite Rotation Mode=::NewEnumerator0
    Sprite Size Mode=::NewEnumerator3
    Uniform Sprite Size=20
    UsePositionOffset=false
    Write Material Random=true
   )
   &AddVelocityInCone(Cone Angle=20, Module.Use Velocity Falloff On Cone Axis=false, Velocity Strength=150)
  ]
  particleUpdateScript [
   &Drag(Drag=&RandomRangeFloat(Minimum=3, Maximum=8), Use Linear Drag=true, Use Rotational Drag=false)
   &GravityForce(Gravity=<0 0 -50>)
   &InheritSourceMovement(
    Applied Position Scale=&VectorFromFloat(Value=&FloatFromCurve(FloatCurve=DataInterface() {propertyValues {Curve {keys [
          {arriveTangent=-14.1181, interpMode=RCIM_Cubic, leaveTangent=-14.1181, leaveTangentWeight=0.3436, tangentMode=RCTM_User, time=0, value=1}
          {arriveTangentWeight=0.3436, interpMode=RCIM_Cubic, tangentMode=RCTM_User, time=0.25, value=0}
         ]}}}))
    ApplyForce=false
    ApplyPosition=true
    ApplyPositionFalloff=false
   )
   &SolveForcesAndVelocity(Clamp Velocity=false, Limit Acceleration=false)
   &ParticleState()
   &SpriteRotationRate(Rotation Rate=&Multiply_Float(
     A=$Particles.InitSpriteRotRate
     B=&FloatFromCurve(
      FloatCurve=DataInterface() {propertyValues {Curve {keys [
          {arriveTangent=-4.0107, interpMode=RCIM_Cubic, leaveTangent=-4.0107, tangentMode=RCTM_User, time=-0.0046, value=1.0113}
          {arriveTangent=-0.2496, interpMode=RCIM_Cubic, leaveTangent=-0.2496, tangentMode=RCTM_User, time=0.5, value=0.16}
         ]}}}
      CurveIndex=$Emitter.Age
      Scale Curve=150
     )
    ))
   &ScaleColor(
    Scale Alpha=&FloatFromCurve(FloatCurve=DataInterface() {propertyValues {Curve {keys [
         {arriveTangent=6.6409, interpMode=RCIM_Cubic, leaveTangent=6.6409, tangentMode=RCTM_User, time=0, value=0}
         {arriveTangentWeight=0.1274, interpMode=RCIM_Cubic, leaveTangentWeight=0.7778, tangentMode=RCTM_User, time=0.2, value=0.9847}
         {arriveTangent=-0.0405, interpMode=RCIM_Cubic, leaveTangent=-0.0405, tangentMode=RCTM_User, time=1, value=0}
        ]}}})
    Scale RGB=&VectorFromFloat(Value=&FloatFromCurve(FloatCurve=DataInterface() {propertyValues {Curve {keys [{arriveTangent=1.8145, interpMode=RCIM_Cubic, leaveTangent=4.5795, tangentMode=RCTM_Break, time=0, value=0.4}, <at 0.3387 0.9565>, <at 1 1>]}}}))
    ScaleA=true
    ScaleRGB=true
   )
   &ScaleSpriteSize(Scale Factor=&Vector2DFromFloat(Value=&FloatFromCurve(FloatCurve=DataInterface() {propertyValues {Curve {keys [
          {arriveTangent=5.6752, interpMode=RCIM_Cubic, leaveTangent=5.6752, leaveTangentWeight=0.9016, tangentMode=RCTM_User, time=0, value=0}
          {arriveTangent=0.1638, interpMode=RCIM_Cubic, leaveTangent=0.1638, tangentMode=RCTM_User, time=1, value=3}
         ]}}})))
   &DynamicMaterialParameters(
    Index 0 Param 1=&Add_Float(
     A=$Particles.RandomFloat
     B=&FloatFromCurve(FloatCurve=DataInterface() {propertyValues {Curve {keys [
          {arriveTangent=2.7533, interpMode=RCIM_Cubic, leaveTangent=2.847, tangentMode=RCTM_Break, time=0, value=0}
          {arriveTangent=1.3581, interpMode=RCIM_Cubic, leaveTangent=1.3581, tangentMode=RCTM_User, time=0.2558, value=0.5125}
          {arriveTangent=0.0529, interpMode=RCIM_Cubic, leaveTangent=0.2021, tangentMode=RCTM_User, time=1, value=1}
         ]}}})
    )
    Index 0 Param 3=&FloatFromCurve(FloatCurve=DataInterface() {propertyValues {Curve {keys [{arriveTangent=-53.9199, time=0, value=0}, {leaveTangent=0.2723, time=0.25, value=0}]}}})
    Index 0 Param 4=&FloatFromCurve(FloatCurve=DataInterface() {propertyValues {Curve {keys [<at 0 1>,<at 1 0>]}}})
    Index 1 Param 1=&FloatFromCurve(FloatCurve=DataInterface() {propertyValues {Curve {keys [<at 0 0>, {arriveTangent=4.2257, interpMode=RCIM_Cubic, leaveTangent=4.2257, tangentMode=RCTM_User, time=0.25, value=0}]}}})
    Index 1 Param 2=&FloatFromCurve(FloatCurve=DataInterface() {propertyValues {Curve {keys [<at 0 0>]}}})
    Index 1 Param 3=&FloatFromCurve(FloatCurve=DataInterface() {propertyValues {Curve {keys [{interpMode=RCIM_Cubic, leaveTangent=6.3917, tangentMode=RCTM_Break, time=0, value=0}]}}})
    Index 1 Param 4=&FloatFromCurve(FloatCurve=DataInterface() {propertyValues {Curve {keys [{arriveTangent=-0.136, leaveTangent=-0.136, time=0, value=1}]}}})
    Index 2 Param 1=&FloatFromCurve(FloatCurve=DataInterface() {propertyValues {Curve {keys [
         {arriveTangent=2.646, interpMode=RCIM_Cubic, leaveTangent=2.646, tangentMode=RCTM_User, time=0, value=0}
         {leaveTangent=-0.0719, time=1, value=1}
        ]}}})
    Index 2 Param 2=&FloatFromCurve(
     FloatCurve=DataInterface() {propertyValues {Curve {keys [
         {arriveTangent=1.9056, time=0, value=0.2}
         {arriveTangent=0.0259, interpMode=RCIM_Cubic, leaveTangent=0.0259, tangentMode=RCTM_User, time=1, value=1.4}
        ]}}}
     Scale Curve=&RandomRangeFloat(Minimum=0.5, Maximum=1)
    )
    Index 2 Param 3=&FloatFromCurve(FloatCurve=DataInterface() {propertyValues {Curve {keys [
         {interpMode=RCIM_Cubic, leaveTangent=2.3064, tangentMode=RCTM_Break, time=0, value=1}
         {interpMode=RCIM_Cubic, leaveTangent=11.1233, tangentMode=RCTM_Break, time=1, value=2}
        ]}}})
    Index 2 Param 4=&FloatFromCurve(
     FloatCurve=DataInterface() {propertyValues {Curve {keys [
         {arriveTangent=-0.6867, interpMode=RCIM_Cubic, leaveTangent=-0.6867, tangentMode=RCTM_User, time=0, value=1}
         {interpMode=RCIM_Cubic, leaveTangent=-1.2956, tangentMode=RCTM_Break, time=1, value=0.75}
        ]}}}
     Scale Curve=1.6
    )
    Module.Param0WriteEnabled=true
    Module.Param0WriteEnabled1=true
    Module.Param0WriteEnabled2=true
    Module.Param1WriteEnabled=false
    Module.Param1WriteEnabled1=true
    Module.Param1WriteEnabled2=true
    Module.Param2WriteEnabled=true
    Module.Param2WriteEnabled1=true
    Module.Param2WriteEnabled2=true
    Module.Param3WriteEnabled=true
    Module.Param3WriteEnabled1=true
    Module.Param3WriteEnabled2=true
    Write Parameter Index 1=true
    Write Parameter Index 2=true
   )
   /Game/Effects/Niagara/Emitters/Water/FN_DistanceKillParticles(AgeReduiceAmount=0.1)
  ]
  rendererClasses [&NiagaraSpriteRendererProperties]
  renderers [
   {rendererClass=&NiagaraSpriteRendererProperties}
   {rendererClass=&NiagaraSpriteRendererProperties, rendererIndex=1}
  ]
 }
 FrostSprites {
  emitterUpdateScript [
   &EmitterState(Life Cycle Mode=::NewEnumerator1, Loop Delay=0.1, Loop Duration=1, UseLoopDelay=true)
   &SpawnRate(SpawnRate=30, Use Spawn Probability=false)
  ]
  particleSpawnScript [
   &InitializeParticle(
    Color=<rgba 0.9404 0.978 1 0.25>
    Lifetime=&RandomRangeFloat(Minimum=0.7, Maximum=1)
    Sprite Rotation Mode=::NewEnumerator2
    Sprite Size Mode=::NewEnumerator1
    Sprite UV Mode=::NewEnumerator3
    Uniform Sprite Size Max=50
    Uniform Sprite Size Min=60
    UsePositionOffset=false
    Write Material Random=true
   )
   &AddVelocity(
    Rotation Coordinate Space=::Simulation
    Velocity=&TransformVector(Vector=<1 0 0>, Destination Space=::World)
    Velocity Speed Scale=&Multiply_Float(A=&RandomRangeFloat(Minimum=380, Maximum=400), B=1)
   )
   &ApplyInitialForces()
   /SetVariables() {isSetParameters=true}
   &CameraOffset(Camera Offset Amount=15)
  ]
  particleUpdateScript [
   &ParticleState()
   &ScaleSpriteSize(Uniform Curve Sprite Scale=DataInterface() {propertyValues {Curve {keys [
        {arriveTangent=0.4427, interpMode=RCIM_Cubic, leaveTangent=0.4427, tangentMode=RCTM_User, time=0, value=0.2}
        {arriveTangent=2.9153, interpMode=RCIM_Cubic, leaveTangent=2.9153, tangentMode=RCTM_User, time=0.4011, value=0.7756}
        <at 1 1.3821 RCIM_Cubic>
       ]}}})
   &ScaleColor(
    Scale Alpha=&Multiply_Float(
     A=&FloatFromCurve(
      FloatCurve=DataInterface() {propertyValues {Curve {keys [
          {arriveTangent=14.1437, interpMode=RCIM_Cubic, leaveTangent=14.1437, tangentMode=RCTM_User, time=0, value=0}
          {
           arriveTangent=-3.0471
           interpMode=RCIM_Cubic
           leaveTangent=-3.0471
           leaveTangentWeight=0.647
           tangentMode=RCTM_User
           time=0.2714
           value=1.9918
          }
          {arriveTangent=-3.2131, interpMode=RCIM_Cubic, leaveTangent=-3.2131, tangentMode=RCTM_User, time=1, value=0}
         ]}}}
      Scale Curve=&FloatFromCurve(
       FloatCurve=DataInterface() {propertyValues {Curve {keys [
           {arriveTangent=2.2094, interpMode=RCIM_Cubic, leaveTangent=2.2094, tangentMode=RCTM_User, time=0, value=0}
           {arriveTangent=1.1179, interpMode=RCIM_Cubic, leaveTangent=1.1179, tangentMode=RCTM_User, time=0.5, value=0.6}
           {arriveTangent=0.0224, interpMode=RCIM_Cubic, leaveTangent=0.0224, tangentMode=RCTM_User, time=1, value=1}
          ]}}}
       CurveIndex=$Emitter.Age
      )
     )
     B=&FloatFromCurve(FloatCurve=DataInterface() {propertyValues {Curve {keys [
          {tangentMode=RCTM_User, time=0, value=1}
          {arriveTangent=-0.0405, interpMode=RCIM_Cubic, leaveTangent=-0.0405, tangentMode=RCTM_User, time=0.2975, value=1}
          {arriveTangent=-6.6903, interpMode=RCIM_Cubic, leaveTangent=-6.6903, tangentMode=RCTM_User, time=0.6668, value=0}
         ]}}})
    )
    ScaleA=true
    ScaleRGB=true
   )
   /SetVariables(Particles.SubImageIndex=&FloatFromCurve(FloatCurve=DataInterface() {propertyValues {Curve {keys [<at 0 30>,<at 1 64>]}}})) {isSetParameters=true}
   &ScaleColor(
    Scale Mode=::NewEnumerator1
    Scale RGBA=&MakeVector4_FromLinearColor(LinearColor=&FNTimeOfDayColor(
      Night Color=<rgba 0.539 0.6941 1 0.3>
      Morning Color=<rgba 0.82 0.8878 1 0.3>
      Day Color=<rgba 1 1 1 0.6>
      Evening Color=<rgba 1 0.8495 0.8351 0.3>
     ))
    ScaleRGBA=true
   ) {module=ScaleColor001}
   &DynamicMaterialParameters(
    Index 0 Param 1=3
    Index 0 Param 2=1
    Index 0 Param 3=0.5
    Module.Index0Param0Write=true
    Module.Index0Param1Write=true
    Module.Index0Param2Write=true
    Module.Index0Param3Write=true
   )
   &Drag(
    Drag=&FloatFromCurve(
     FloatCurve=DataInterface() {propertyValues {Curve {keys [
         {interpMode=RCIM_Cubic, tangentMode=RCTM_User, time=0, value=0}
         {arriveTangent=22.0021, interpMode=RCIM_Cubic, leaveTangent=22.0021, tangentMode=RCTM_User, time=0.06, value=1}
        ]}}}
     Scale Curve=&Multiply_Float(A=5, B=&Subtract_Float(A=2.5, B=0.9))
    )
    Use Linear Drag=true
    Use Rotational Drag=true
   ) {module=Drag002}
   &InheritSourceMovement(
    Applied Position Falloff Distance=1000
    Applied Position Scale=&Multiply_VectorByFloat(
     Vector=<0.6 0.6 0>
     Float=&FloatFromCurve(
      FloatCurve=DataInterface() {propertyValues {Curve {keys [
          <at 0 0 RCIM_Cubic>
          {interpMode=RCIM_Cubic, tangentMode=RCTM_User, time=0.1865, value=1}
          {arriveTangent=-2.1905, interpMode=RCIM_Cubic, leaveTangent=-2.1905, tangentMode=RCTM_User, time=1, value=0}
         ]}}}
      CurveIndex=$Emitter.Age
     )
    )
    ApplyForce=false
    ApplyPosition=true
    ApplyPositionFalloff=false
   )
   &SolveForcesAndVelocity(Clamp Velocity=false, Limit Acceleration=false)
  ]
  rendererClasses [&NiagaraSpriteRendererProperties]
  renderers [{rendererClass=&NiagaraSpriteRendererProperties}]
 }
 Spray_Jet {rendererClasses [], renderers []}
}
summary {
 systemName=NS_VelvetDeskCam_MistSpray
 emitters [
  {emitterName=FN_Core_Smoke_Sprite, rendererClasses [&NiagaraSpriteRendererProperties]}
  {emitterName=FrostSprites, rendererClasses [&NiagaraSpriteRendererProperties]}
  {emitterName=Spray_Jet, rendererClasses []}
  {emitterName=Cone, rendererClasses []}
 ]
 userVariables []
}
