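# Niagara effect for the VelvetDeskCam emote that produces a brief puff of mist, as for a poke.
# GPU MistChunks bursts two sprites aligned to a fixed orientation, which grow, spin with decaying speed and fade under drag.
# Smoke bursts camera-offset SubUV sprites from random start frames, which drift, stretch and fade at low opacity.
# Both emitters tint and scale spawn color by time of day and drive material look through dynamic parameters.
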NiagaraSystem /BRCosmetics/Animation/Game/MainPlayer/Emotes/VelvetDeskCam/FX/NS_VelvetDeskCam_PokeMist
dependencies {
 usedDataInterfaces [
  &NiagaraDataInterfaceCurve
  &NiagaraDataInterfaceVector2DCurve
  &NiagaraDataInterfaceSpriteRendererInfo
  &NiagaraDataInterfaceColorCurve
 ]
 usedDynamicInputs []
 usedModules [&SystemState]
}
emitters {
 MistChunks {
  simTarget=GPUComputeSim
  emitterUpdateScript [
   &EmitterState(Life Cycle Mode=::NewEnumerator1, Loop Behavior=::NewEnumerator1, Loop Duration=1, UseLoopDelay=false)
   &SpawnBurst_Instantaneous(Spawn Count=2, Use Spawn Probability=false, UseLoopCountLimit=false)
  ]
  particleSpawnScript [
   &InitializeParticle(
    Color=&Multiply_LinearColorByFloat(
     Linear Color=&FNTimeOfDayColor(Night Color=<rgba 0.4977 0.6355 1 1>, Morning Color=<rgba 1 0.5989 0.7062 1>, Day Color=<rgba 1 1 1 1>, Evening Color=<rgba 1 1 1 1>)
     Float=&Multiply_Float(A=1, B=&FNTimeOfDayScalar(Night Scale=0.7, Morning Scale=0.8, Evening Scale=0.8))
    )
    Lifetime Max=1
    Lifetime Min=0.5
    Lifetime Mode=::NewEnumerator1
    Mass Max=1.5
    Mass Mode=::NewEnumerator3
    Sprite Rotation Mode=::NewEnumerator2
    Sprite Size Mode=::NewEnumerator1
    Uniform Sprite Size Max=16
    Uniform Sprite Size Min=10
    UsePositionOffset=false
    Write Material Random=true
   )
   /SetVariables(Particles.InitSpriteRotRate=&RandomRangeFloat(Minimum=0.3, Maximum=-0.4)) {isSetParameters=true}
   &ApplyInitialForces()
   &InitialMeshOrientation(Mesh Orientation Mode=::NewEnumerator2, Orientation Vector=<0 -1 0>, Use Rotation Vector=true)
   &AlignSpriteToMeshOrientation(SetAlignment=true, SetFacing=true)
  ]
  particleUpdateScript [
   &ParticleState()
   &Drag(Drag=&RandomRangeFloat(Minimum=1.8, Maximum=2.5), Use Linear Drag=true, Use Rotational Drag=true)
   &SolveForcesAndVelocity(Clamp Velocity=false, Limit Acceleration=false)
   &ScaleSpriteSize(Scale Factor=&Vector2DFromFloat(Value=&FloatFromCurve(
      FloatCurve=DataInterface() {propertyValues {Curve {keys [
          {arriveTangent=3.2437, interpMode=RCIM_Cubic, leaveTangent=3.2437, leaveTangentWeight=0.9016, tangentMode=RCTM_User, time=0, value=0.2}
          {arriveTangent=0.1638, interpMode=RCIM_Cubic, leaveTangent=0.1638, tangentMode=RCTM_User, time=1, value=1}
         ]}}}
      Scale Curve=2.5
     )))
   &SpriteRotationRate(Rotation Rate=&Multiply_Float(
     A=$Particles.InitSpriteRotRate
     B=&FloatFromCurve(
      FloatCurve=DataInterface() {propertyValues {Curve {keys [
          {arriveTangent=-4.0107, interpMode=RCIM_Cubic, leaveTangent=-4.0107, tangentMode=RCTM_User, time=-0.0046, value=1.0113}
          {arriveTangent=-0.2496, interpMode=RCIM_Cubic, leaveTangent=-0.2496, tangentMode=RCTM_User, time=0.5, value=0.16}
         ]}}}
      CurveIndex=$Emitter.Age
      Scale Curve=150
     )
    ))
   &DynamicMaterialParameters(
    Index 0 Param 1=1
    Index 0 Param 2=0.4
    Index 0 Param 3=0.6
    Index 0 Param 4=4
    Index 1 Param 1=1
    Index 1 Param 2=3
    Index 3 Param 1=-2
    Index 3 Param 2=1
    Index 3 Param 3=1
    Index 3 Param 4=3
    Module.Index0Float0Write=true
    Module.Index0Float1Write=true
    Module.Index0Float2Write=true
    Module.Index0Float3Write=true
    Module.Index1Float0Write=true
    Module.Index1Float1Write=true
    Module.Index1Float2Write=true
    Module.Index1Float3Write=true
    Module.Index3Float0Write=true
    Module.Index3Float1Write=true
    Module.Index3Float2Write=true
    Module.Index3Float3Write=true
    Write Parameter Index 1=::NewEnumerator1
    Write Parameter Index 3=::NewEnumerator1
   )
   &ScaleColor(
    Scale Alpha=&FloatFromCurve(FloatCurve=DataInterface() {propertyValues {Curve {keys [
         {arriveTangent=6.6409, interpMode=RCIM_Cubic, leaveTangent=6.6409, tangentMode=RCTM_User, time=0, value=0}
         {arriveTangentWeight=0.1274, interpMode=RCIM_Cubic, leaveTangentWeight=0.7778, tangentMode=RCTM_User, time=0.0801, value=0.9847}
         {arriveTangent=-0.0405, interpMode=RCIM_Cubic, leaveTangent=-0.0405, tangentMode=RCTM_User, time=1, value=0}
        ]}}})
    Scale RGB=&VectorFromFloat(Value=&FloatFromCurve(FloatCurve=DataInterface() {propertyValues {Curve {keys [{arriveTangent=1.8145, interpMode=RCIM_Cubic, leaveTangent=4.5795, tangentMode=RCTM_Break, time=0, value=0.4}, <at 0.3387 0.9565>, <at 1 1>]}}}))
    ScaleA=true
    ScaleRGB=true
   )
  ]
  rendererClasses [&NiagaraSpriteRendererProperties]
  renderers [{rendererClass=&NiagaraSpriteRendererProperties}]
 }
 Smoke {
  emitterUpdateScript [
   &EmitterState(Life Cycle Mode=::NewEnumerator1, Loop Behavior=::NewEnumerator1, Loop Duration=3, UseLoopDelay=false)
   &SpawnBurst_Instantaneous(Use Spawn Probability=false, UseLoopCountLimit=false)
  ]
  particleSpawnScript [
   &InitializeParticle(
    Color=&Multiply_LinearColorByFloat(
     Linear Color=&FNTimeOfDayColor(Night Color=<rgba 0.4977 0.6355 1 1>, Morning Color=<rgba 1 0.5989 0.7062 1>, Day Color=<rgba 1 1 1 1>, Evening Color=<rgba 1 1 1 1>)
     Float=&FNTimeOfDayScalar(Night Scale=0.5, Morning Scale=1, Evening Scale=1)
    )
    Lifetime=&MaskFloatBySpawnGroup(Masked Float=1, Passthrough Float=0.8)
    Position Offset=<0 10 0>
    Sprite Rotation Angle=0.5
    Sprite Rotation Mode=::NewEnumerator1
    Sprite Size Mode=::NewEnumerator3
    Sprite UV Mode=::NewEnumerator4
    Uniform Sprite Size=15
    UsePositionOffset=true
    Write Material Random=true
   )
   &PointForce(
    Force Strength=&RandomRangeFloat(Minimum=1, Maximum=1)
    Linear Force Vector:Vector3f=<0 1 0>
    Module.Use Falloff=false
    Module.UseLinearForce=true
   )
   &CameraOffset(Camera Offset Amount=30)
   &DynamicMaterialParameters(
    Index 0 Param 3=1
    Module.Index0Float0Write=true
    Module.Index0Float1Write=true
    Module.Index0Float2Write=true
    Module.Index0Float3Write=true
   )
   &ApplyInitialForces()
  ]
  particleUpdateScript [
   &ParticleState()
   &ScaleSpriteSize(
    Scale Factor=&Vector2DFromCurve(
     Vector2Curve=DataInterface() {propertyValues {
       XCurve {keys [<at 0 1>,<at 1 1>]}
       YCurve {keys [
         {arriveTangent=1.2325, interpMode=RCIM_Cubic, leaveTangent=1.2325, tangentMode=RCTM_User, time=0, value=0.58}
         {interpMode=RCIM_Cubic, tangentMode=RCTM_User, time=1, value=1}
        ]}
      }}
     ScaleCurve=&MaskVector2dBySpawnGroup(Masked Vector2D=<1.5 3>, Passthrough Vector2D=<1.5 1.8>, Spawn Group Mask=2)
    )
    Scale Sprite Size Mode=::NewEnumerator2
   )
   &SubUVAnimation(
    End Frame Range Override=30
    "Sprite Renderer "=DataInterface() {propertyValues {SpriteRenderer=&self:Smoke_0.NiagaraSpriteRendererProperties_2}}
    Start Frame Range Override=&RandomRangeInt(Minimum=0, Maximum=5)
    SubUV Lookup Index Scale=0.4
    UseEndFrame=false
    UseStartFrame=true
   )
   &Drag(
    Drag=&FloatFromCurve(FloatCurve=DataInterface() {propertyValues {Curve {keys [{interpMode=RCIM_Cubic, tangentMode=RCTM_User, time=0, value=0}, {interpMode=RCIM_Cubic, tangentMode=RCTM_User, time=0.6258, value=1}]}}})
    Use Linear Drag=true
    Use Rotational Drag=true
   )
   &SolveForcesAndVelocity(Clamp Velocity=false, Limit Acceleration=false)
   &ScaleColor(
    Scale Alpha=&FloatFromCurve(
     FloatCurve=DataInterface() {propertyValues {Curve {keys [
         {arriveTangent=4.7901, interpMode=RCIM_Cubic, leaveTangent=4.7901, tangentMode=RCTM_User, time=0, value=0}
         {interpMode=RCIM_Cubic, tangentMode=RCTM_User, time=0.5, value=1}
         {tangentMode=RCTM_User, time=1, value=0}
        ]}}}
     Scale Curve=0.4
    )
    ScaleA=true
    ScaleRGB=true
   ) {module=ScaleColor001}
  ]
  rendererClasses [&NiagaraSpriteRendererProperties]
  renderers [{rendererClass=&NiagaraSpriteRendererProperties}]
 }
}
summary {
 systemName=NS_VelvetDeskCam_PokeMist
 emitters [
  {emitterName=MistChunks, simTarget=GPUComputeSim, rendererClasses [&NiagaraSpriteRendererProperties]}
  {emitterName=Smoke, rendererClasses [&NiagaraSpriteRendererProperties]}
 ]
 userVariables []
}
