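# Niagara effect for the VelvetDeskCam emote that shows a smiley pinned to the gadget's JNT_Spray_FX bone, then dissipates it into smoke.
# Main renders two meshes, with age-timed curves driving erosion reveal, eye stretch, WPO, squash and fade.
# Smoke fires three delayed SubUV sprite bursts as the face breaks up.
# Camera-to-system-axis dot products fade edge and smoke layers, time of day adjusts color, and SmileSize scales both emitters.
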NiagaraSystem /BRCosmetics/Animation/Game/MainPlayer/Emotes/VelvetDeskCam/FX/NS_VelvetDeskCam_Smiley
dependencies {
 usedDataInterfaces [
  &NiagaraDataInterfaceCurve
  &NiagaraDataInterfacePlatformSet
  &NiagaraDataInterfaceSkeletalMesh
  &NiagaraDataInterfaceVectorCurve
  &NiagaraDataInterfaceCamera
  &NiagaraDataInterfaceArrayFloat
  &NiagaraDataInterfaceVector2DCurve
  &NiagaraDataInterfaceSpriteRendererInfo
  &NiagaraDataInterfaceColorCurve
 ]
 usedDynamicInputs []
 usedModules [&SystemState]
}
emitters {
 Main {
  emitterUpdateScript [
   &EmitterState(Life Cycle Mode=::NewEnumerator1, Loop Behavior=::NewEnumerator1, UseLoopDelay=false)
   &SpawnBurst_Instantaneous(Use Spawn Probability=false, UseLoopCountLimit=false)
   /SetVariables(
    Emitter.Erosions:Vector3f=&MakeVector(
     X=&FloatFromCurve(FloatCurve=DataInterface() {propertyValues {Curve {keys [<at 1.7 0>,<at 1.8 1>]}}}, CurveIndex=$Emitter.Age, Scale Curve=5)
     Y=&FloatFromCurve(FloatCurve=DataInterface() {propertyValues {Curve {keys [<at 1.26 0>,<at 1.36 1>]}}}, CurveIndex=$Emitter.Age, Scale Curve=5)
     Z=&FloatFromCurve(FloatCurve=DataInterface() {propertyValues {Curve {keys [<at 1.94 0>,<at 2.13 1>]}}}, CurveIndex=$Emitter.Age)
    )
    Emitter.SmileyEdgeMask:NiagaraFloat=&FN_PickFloatDependOnPlatform(TargetPlatformActive=1, TargetPlatformNoActive=0)
   ) {isSetParameters=true}
   /SetVariables(
    Emitter.EdgesErosions:Vector3f=&MakeVector(
     X=&FloatFromCurve(FloatCurve=DataInterface() {propertyValues {Curve {keys [<at 1.7 0>,<at 1.8 1>]}}}, CurveIndex=$Emitter.Age, Scale Curve=5)
     Y=&FloatFromCurve(FloatCurve=DataInterface() {propertyValues {Curve {keys [<at 1.26 0>,<at 1.36 1>]}}}, CurveIndex=$Emitter.Age, Scale Curve=5)
     Z=&FloatFromCurve(FloatCurve=DataInterface() {propertyValues {Curve {keys [<at 1.94 0>,<at 2.13 1>]}}}, CurveIndex=$Emitter.Age)
    )
    Emitter.EdgesInvertSideMask=1
    Emitter.EdgesSmileyEdgeMask=1
   ) {isSetParameters=true}
   /SetVariables(
    Emitter.Eye1StretchOffset:NiagaraFloat=&FloatFromCurve(
     FloatCurve=DataInterface() {propertyValues {Curve {keys [
         {arriveTangent=1.0069, arriveTangentWeight=1.1667, interpMode=RCIM_Cubic, leaveTangent=1.0069, tangentMode=RCTM_User, time=4.92, value=0}
         {arriveTangent=-0.0218, arriveTangentWeight=0.2333, interpMode=RCIM_Cubic, leaveTangent=-0.0218, tangentMode=RCTM_User, time=5.25, value=0.09}
        ]}}}
     CurveIndex=$Emitter.Age
    )
    Emitter.Eye1StretchSize:NiagaraFloat=&FloatFromCurve(
     FloatCurve=DataInterface() {propertyValues {Curve {keys [
         {arriveTangent=-3.4714, arriveTangentWeight=1.1667, interpMode=RCIM_Cubic, leaveTangent=-3.4714, tangentMode=RCTM_User, time=4.92, value=1}
         {arriveTangent=-0.0218, arriveTangentWeight=0.2333, interpMode=RCIM_Cubic, leaveTangent=-0.0218, tangentMode=RCTM_User, time=5.25, value=0.7}
        ]}}}
     CurveIndex=$Emitter.Age
    )
    Emitter.Eye2StretchOffset:NiagaraFloat=&FloatFromCurve(
     FloatCurve=DataInterface() {propertyValues {Curve {keys [
         {arriveTangent=1.0069, arriveTangentWeight=1.1667, interpMode=RCIM_Cubic, leaveTangent=1.0069, tangentMode=RCTM_User, time=4.5, value=0}
         {arriveTangent=-0.0218, arriveTangentWeight=0.2333, interpMode=RCIM_Cubic, leaveTangent=-0.0218, tangentMode=RCTM_User, time=5.2, value=0.25}
        ]}}}
     CurveIndex=$Emitter.Age
    )
    Emitter.Eye2StretchSize:NiagaraFloat=&FloatFromCurve(
     FloatCurve=DataInterface() {propertyValues {Curve {keys [
         {arriveTangent=-1.3827, arriveTangentWeight=1.1667, interpMode=RCIM_Cubic, leaveTangent=-1.3827, tangentMode=RCTM_User, time=4.5, value=1}
         {arriveTangent=-0.0218, arriveTangentWeight=0.2333, interpMode=RCIM_Cubic, leaveTangent=-0.0218, tangentMode=RCTM_User, time=5.2, value=0.6}
        ]}}}
     CurveIndex=$Emitter.Age
    )
    Emitter.GlobalOpacity:NiagaraFloat=&FloatFromCurve(
     FloatCurve=DataInterface() {propertyValues {Curve {keys [{arriveTangent=178.1007, interpMode=RCIM_Cubic, time=4.5, value=1}, <at 5.6 0 RCIM_Cubic>]}}}
     CurveIndex=$Emitter.Age
     Scale Curve=&FNTimeOfDayScalar(Night Scale=1, Morning Scale=1, Day Scale=1.4, Evening Scale=1)
    )
    Emitter.WPOSide1:NiagaraFloat=&FloatFromCurve(
     FloatCurve=DataInterface() {propertyValues {Curve {keys [
         {arriveTangent=-191.5866, interpMode=RCIM_Cubic, leaveTangent=-191.5866, tangentMode=RCTM_User, time=4.5, value=0}
         {arriveTangent=-90.9091, interpMode=RCIM_Cubic, leaveTangent=-90.9091, time=5.3, value=-80}
         <at 5.6 -100 RCIM_Cubic>
        ]}}}
     CurveIndex=$Emitter.Age
    )
    Emitter.WPOSide2:NiagaraFloat=&FloatFromCurve(
     FloatCurve=DataInterface() {propertyValues {Curve {keys [
         {arriveTangent=209.5369, interpMode=RCIM_Cubic, leaveTangent=209.5369, tangentMode=RCTM_User, time=4.92, value=0}
         {arriveTangent=20.4082, interpMode=RCIM_Cubic, leaveTangent=20.4082, time=5.25, value=56.25}
         {interpMode=RCIM_Cubic, leaveTangent=27.9412, time=5.9, value=20}
        ]}}}
     CurveIndex=$Emitter.Age
    )
   ) {isSetParameters=true}
  ]
  particleSpawnScript [
   &InitializeParticle(
    Color=&Multiply_LinearColorByFloat(
     Linear Color=&FNTimeOfDayColor(Night Color=<rgba 0.4977 0.6355 1 1>, Morning Color=<rgba 1 0.5989 0.7062 1>, Day Color=<rgba 1 1 1 1>, Evening Color=<rgba 1 1 1 1>)
     Float=&FNTimeOfDayScalar(Night Scale=1, Morning Scale=1, Evening Scale=1)
    )
    Lifetime=6.5
    Mesh Scale=&Multiply_VectorByFloat(Vector=<1 1 1>, Float=$User.SmileSize)
    Mesh Scale Mode=::NewEnumerator4
    UsePositionOffset=false
    Write Material Random=true
   )
   &InitialMeshOrientation(Mesh Orientation Mode=::NewEnumerator2, Orientation Vector=<0 1 0>, Rotation=<0 0 -0.25>, Use Rotation Vector=true)
  ]
  particleUpdateScript [
   &SkeletalMeshLocation(
    Bone Sampling Mode=::NewEnumerator4
    MaskSpawnGroup=false
    Skeletal Mesh=DataInterface() {propertyValues {
      DefaultMesh=/BRCosmetics/Animation/Props/Gadgets_VelvetDeskCam/Meshes/VelvetDeskCam_Gadget
      ExcludeBoneName=root
      OptionalComponentClass=null
      PreviewMesh=&VelvetDeskCam_Gadget
      SkinningMode=SkinOnTheFly
      SoftSourceActor=null
      SourceMode=AttachParent
      UvSetIndex=0
      WholeMeshLOD=-1
      bExcludeBone=false
      bReadDeformedGeometry=true
      bRequireCurrentFrameData=true
      ComponentTags []
      FilteredBones [JNT_Spray_FX]
      FilteredSockets []
      MeshUserParameter {parameter {name=None, typeDefHandle {registeredTypeIndex=1}, varData []}}
      SamplingRegions []
     }}
   )
   &ParticleState()
   &DynamicMaterialParameters(
    Index 0 Param 2=0.2
    Index 0 Param 3=&FloatFromCurve(
     FloatCurve=DataInterface() {propertyValues {Curve {keys [{tangentMode=RCTM_User, time=0, value=0}, {tangentMode=RCTM_User, time=0.2, value=1}]}}}
     Scale Curve=&FNTimeOfDayScalar(Night Scale=1, Morning Scale=1, Evening Scale=1)
    )
    Index 0 Param 4=4
    Index 1 Param 2=&FloatFromCurve(
     FloatCurve=DataInterface() {propertyValues {Curve {keys [{tangentMode=RCTM_User, time=0, value=0}, {tangentMode=RCTM_User, time=0.2, value=1}]}}}
     Scale Curve=&FNTimeOfDayScalar(Night Scale=1, Morning Scale=1, Evening Scale=1)
    )
    Index 3 Param 1=&FloatFromCurve(FloatCurve=DataInterface() {propertyValues {Curve {keys [<at 4.7979 -1>,<at 6.0187 1>]}}}, CurveIndex=$Emitter.Age)
    Index 3 Param 2=&FloatFromCurve(FloatCurve=DataInterface() {propertyValues {Curve {keys [
         {arriveTangent=-2.6933, interpMode=RCIM_Cubic, leaveTangent=-2.6933, tangentMode=RCTM_User, time=0.685, value=1}
         {arriveTangent=-1.3244, interpMode=RCIM_Cubic, leaveTangent=-1.3244, tangentMode=RCTM_User, time=0.7504, value=0.7374}
         <at 1 0.5 RCIM_Cubic>
        ]}}})
    Index 3 Param 3=&FloatFromCurve(FloatCurve=DataInterface() {propertyValues {Curve {keys [<at 0.6497 0.025 RCIM_Cubic>,<at 1 1 RCIM_Cubic>]}}})
    Index 3 Param 4=1
    Module.Index0Param0Write=true
    Module.Index0Param1Write=true
    Module.Index0Param2Write=true
    Module.Index0Param3Write=true
    Module.Index1Param0Write=true
    Module.Index1Param1Write=true
    Module.Index1Param2Write=true
    Module.Index1Param3Write=true
    Module.Index3Param0Write=true
    Module.Index3Param1Write=true
    Module.Index3Param2Write=true
    Module.Index3Param3Write=true
    Write Parameter Index 1=true
    Write Parameter Index 3=true
   )
   /SetVariables(Particles.Scale=&Add_Vector(A=&Multiply_Vector(A=$Particles.SkeletalMeshLocation.SampledScale, B=<0 0 0>), B=$Particles.Initial.Scale)) {isSetParameters=true}
   &ScaleMeshSize(Scale Factor=&VectorFromCurve(VectorCurve=DataInterface() {propertyValues {
       XCurve {keys [<at 0 1>,<at 1 1>]}
       YCurve {keys [<at 0 1>,<at 1 1>]}
       ZCurve {keys [
         {arriveTangent=7.7957, interpMode=RCIM_Cubic, time=0.68, value=1}
         {
          arriveTangent=-5.2876
          arriveTangentWeight=0.1257
          interpMode=RCIM_Cubic
          leaveTangent=-0.4717
          leaveTangentWeight=0.0558
          tangentMode=RCTM_Break
          tangentWeightMode=RCTWM_WeightedBoth
          time=0.715
          value=0.6572
         }
         {
          arriveTangent=-0.4361
          arriveTangentWeight=0.0195
          interpMode=RCIM_Cubic
          leaveTangent=-0.4676
          leaveTangentWeight=0.0747
          tangentMode=RCTM_Break
          tangentWeightMode=RCTWM_WeightedBoth
          time=0.7713
          value=0.475
         }
         {interpMode=RCIM_Cubic, leaveTangent=-3.5146, time=0.8549, value=0.2315}
        ]}
      }}))
   &ScaleColor(
    Scale Alpha=&FloatFromCurve(FloatCurve=DataInterface() {propertyValues {Curve {keys [{interpMode=RCIM_Cubic, tangentMode=RCTM_User, time=0.7043, value=1}, {interpMode=RCIM_Cubic, tangentMode=RCTM_User, time=1, value=0}]}}})
    ScaleA=true
    ScaleRGB=false
   )
   /SetVariables(
    Particles.EdgesColor:LinearColor=&MakeLinearColor(
     R:NiagaraFloat=1
     G:NiagaraFloat=1
     B:NiagaraFloat=1
     A:NiagaraFloat=&Multiply_Float(
      A=1
      B=&ClampFloat(Float=&Lerp_Float(
        A=1
        B=0
        Alpha=&Abs_Float(Float:NiagaraFloat=&DotProduct(
          Vector A=&NormalizeVector(
           Vector To Normalize=&ReturnCameraVector(Camera Vectors=::NewEnumerator1, Coordinate Space=::World)
           Safe Normalize=true
           Safe Fallback Threshold=0
           Safe Fallback Vector=<0 0 0.001>
          )
          Vector B=&NormalizeVector(Vector To Normalize=$Engine.Owner.SystemYAxis, Safe Normalize=true, Safe Fallback Threshold=0, Safe Fallback Vector=<0 0 0.001>)
         ))
       ))
     )
    )
    Particles.EdgesDynamicParameters:Vector4f=&MakeVector4(
     X=&FloatFromCurve(FloatCurve=DataInterface() {propertyValues {Curve {keys [<at 4.7979 -1>,<at 6.0187 1>]}}}, CurveIndex=$Emitter.Age)
     Y=&FloatFromCurve(FloatCurve=DataInterface() {propertyValues {Curve {keys [
          {arriveTangent=-2.6933, interpMode=RCIM_Cubic, leaveTangent=-2.6933, tangentMode=RCTM_User, time=0.685, value=1}
          {arriveTangent=-1.3244, interpMode=RCIM_Cubic, leaveTangent=-1.3244, tangentMode=RCTM_User, time=0.7504, value=0.7374}
          <at 1 0.5 RCIM_Cubic>
         ]}}})
     Z=0.025
     W=1
    )
    Particles.EdgesScale:Vector3f=&VectorFromCurve(
     VectorCurve=DataInterface() {propertyValues {
       XCurve {keys [<at 0 1>,<at 1 1>]}
       YCurve {keys [<at 0 1>,<at 1 1>]}
       ZCurve {keys [
         {arriveTangent=7.7957, interpMode=RCIM_Cubic, time=0.68, value=1}
         {
          arriveTangent=-5.2876
          arriveTangentWeight=0.1257
          interpMode=RCIM_Cubic
          leaveTangent=-0.4717
          leaveTangentWeight=0.0558
          tangentMode=RCTM_Break
          tangentWeightMode=RCTWM_WeightedBoth
          time=0.715
          value=0.6572
         }
         {
          arriveTangent=-0.4361
          arriveTangentWeight=0.0195
          interpMode=RCIM_Cubic
          leaveTangent=-0.4676
          leaveTangentWeight=0.0747
          tangentMode=RCTM_Break
          tangentWeightMode=RCTWM_WeightedBoth
          time=0.7713
          value=0.475
         }
         {interpMode=RCIM_Cubic, leaveTangent=-3.5146, time=0.8549, value=0.2315}
        ]}
      }}
     Scale Curve=&Multiply_VectorByFloat(Vector=<0.9 1.5 0.9>, Float=$User.SmileSize)
    )
   ) {isSetParameters=true}
  ]
  rendererClasses [&NiagaraMeshRendererProperties]
  renderers [
   {rendererClass=&NiagaraMeshRendererProperties, rendererIndex=0}
   {rendererClass=&NiagaraMeshRendererProperties, rendererIndex=1}
  ]
 }
 Smoke {
  emitterUpdateScript [
   &EmitterState(Life Cycle Mode=::NewEnumerator1, Loop Behavior=::NewEnumerator1, Loop Delay=4.5, Loop Duration=3, UseLoopDelay=true)
   &SpawnBurst_Instantaneous(Spawn Group=2, Spawn Time=0.72, Use Spawn Probability=false, UseLoopCountLimit=false) {module=SpawnBurst_Instantaneous002}
   &SpawnBurst_Instantaneous(Use Spawn Probability=false, UseLoopCountLimit=false) {module=SpawnBurst_Instantaneous001}
   &SpawnBurst_Instantaneous(Spawn Group=1, Spawn Time=0.4, Use Spawn Probability=false, UseLoopCountLimit=false)
  ]
  particleSpawnScript [
   &InitializeParticle(
    Color=&Multiply_LinearColorByFloat(
     Linear Color=&FNTimeOfDayColor(Night Color=<rgba 0.4977 0.6355 1 1>, Morning Color=<rgba 1 0.5989 0.7062 1>, Day Color=<rgba 1 1 1 1>, Evening Color=<rgba 1 1 1 1>)
     Float=&FNTimeOfDayScalar(Night Scale=0.7, Morning Scale=1, Evening Scale=1)
    )
    Lifetime=&MaskFloatBySpawnGroup(Masked Float=1, Passthrough Float=0.8)
    Sprite Rotation Angle=0.6696
    Sprite Rotation Mode=::NewEnumerator1
    Sprite Size Mode=::NewEnumerator3
    Sprite UV Mode=::NewEnumerator4
    Uniform Sprite Size=&Multiply_Float(A=&RandomRangeFloat(Minimum=40, Maximum=30), B=$User.SmileSize)
    UsePositionOffset=false
    Write Material Random=true
   )
   &CameraOffset(Camera Offset Amount=30)
   &AddVelocity(
    Velocity=&MaskVector(
     Mask Parameter=$Emitter.SpawnGroup
     Mask=0
     Masked Vector=<80 0 0>
     Passthrough Vector=&MaskVectorBySpawnGroup(Masked Vector=<-50 0 0>, Passthrough Vector=<80 0 0>, Spawn Group Mask=1)
    )
    Velocity Speed Scale=0.1
   )
   &DynamicMaterialParameters(
    Index 0 Param 3=1
    Module.Index0Float0Write=true
    Module.Index0Float1Write=true
    Module.Index0Float2Write=true
    Module.Index0Float3Write=true
   )
   /SetVariables(Particles.SideRotation:NiagaraFloat=&SelectFloatFromArray(
     Array Sampling Mode=::NewEnumerator0
     Direct Array Index=?
     Float Selection Array=DataInterface() {propertyValues {FloatData [0.48,0.5,0.52,0,0.03,0.98]}}
     Random Seed=0
    )) {isSetParameters=true}
  ]
  particleUpdateScript [
   &ParticleState()
   &SkeletalMeshLocation(
    Bone Sampling Mode=::NewEnumerator4
    MaskSpawnGroup=false
    Position Sampling=::NewEnumerator1
    Sampled Position Offset=&MaskVectorBySpawnGroup(
     Masked Vector=<0 0 0>
     Passthrough Vector=&MaskVectorBySpawnGroup(Masked Vector=<-10 0 0>, Passthrough Vector=<20 0 0>, Spawn Group Mask=1)
    )
    Skeletal Mesh=DataInterface() {propertyValues {
      DefaultMesh=&VelvetDeskCam_Gadget
      ExcludeBoneName=root
      OptionalComponentClass=null
      PreviewMesh=&VelvetDeskCam_Gadget
      SkinningMode=SkinOnTheFly
      SoftSourceActor=null
      SourceMode=AttachParent
      UvSetIndex=0
      WholeMeshLOD=-1
      bExcludeBone=false
      bReadDeformedGeometry=true
      bRequireCurrentFrameData=true
      ComponentTags []
      FilteredBones [JNT_Spray_FX]
      FilteredSockets []
      MeshUserParameter {parameter {name=None, typeDefHandle {registeredTypeIndex=1}, varData []}}
      SamplingRegions []
     }}
   )
   &ScaleSpriteSize(
    Scale Factor=&Vector2DFromCurve(
     Vector2Curve=DataInterface() {propertyValues {
       XCurve {keys [<at 0 1>,<at 1 1>]}
       YCurve {keys [
         {arriveTangent=1.2325, interpMode=RCIM_Cubic, leaveTangent=1.2325, tangentMode=RCTM_User, time=0, value=0.58}
         {interpMode=RCIM_Cubic, tangentMode=RCTM_User, time=1, value=1}
        ]}
      }}
     ScaleCurve=&MaskVector2dBySpawnGroup(Masked Vector2D=<1.5 3>, Passthrough Vector2D=<1.5 1.8>, Spawn Group Mask=2)
    )
    Scale Sprite Size Mode=::NewEnumerator2
   )
   &SubUVAnimation(
    End Frame Range Override=30
    "Sprite Renderer "=DataInterface() {propertyValues {SpriteRenderer=&self:Smoke_0.NiagaraSpriteRendererProperties_2}}
    Start Frame Range Override=&RandomRangeInt(Minimum=0, Maximum=5)
    SubUV Lookup Index Scale=0.4
    UseEndFrame=false
    UseStartFrame=true
   )
   &Drag(
    Drag=&FloatFromCurve(FloatCurve=DataInterface() {propertyValues {Curve {keys [{interpMode=RCIM_Cubic, tangentMode=RCTM_User, time=0, value=0}, {interpMode=RCIM_Cubic, tangentMode=RCTM_User, time=0.6258, value=1}]}}})
    Use Linear Drag=true
    Use Rotational Drag=true
   )
   &SolveForcesAndVelocity(Clamp Velocity=false, Limit Acceleration=false)
   &ScaleColor(
    Scale Alpha=&Multiply_Float(
     A=&FloatFromCurve(
      FloatCurve=DataInterface() {propertyValues {Curve {keys [
          {arriveTangent=4.7901, interpMode=RCIM_Cubic, leaveTangent=4.7901, tangentMode=RCTM_User, time=0, value=0}
          {interpMode=RCIM_Cubic, tangentMode=RCTM_User, time=0.5, value=1}
          {tangentMode=RCTM_User, time=1, value=0}
         ]}}}
      Scale Curve=&MaskFloatBySpawnGroup(Masked Float=0.3, Passthrough Float=0.1)
     )
     B=&Lerp_Float(
      A=1
      B=0
      Alpha=&Abs_Float(Float:NiagaraFloat=&DotProduct(
        Vector A=&NormalizeVector(
         Vector To Normalize=&ReturnCameraVector(Camera Vectors=::NewEnumerator1, Coordinate Space=::World)
         Safe Normalize=true
         Safe Fallback Threshold=0
         Safe Fallback Vector=<0 0 0.001>
        )
        Vector B=&NormalizeVector(Vector To Normalize=$Engine.Owner.SystemXAxis, Safe Normalize=true, Safe Fallback Threshold=0, Safe Fallback Vector=<0 0 0.001>)
       ))
     )
    )
    ScaleA=true
    ScaleRGB=true
   ) {module=ScaleColor001}
   /SetVariables(Particles.SideColor:LinearColor=&MakeLinearColor(
     R:NiagaraFloat=1
     G:NiagaraFloat=1
     B:NiagaraFloat=1
     A:NiagaraFloat=&Multiply_Float(
      A=&FloatFromCurve(
       FloatCurve=DataInterface() {propertyValues {Curve {keys [
           {arriveTangent=4.7901, interpMode=RCIM_Cubic, leaveTangent=4.7901, tangentMode=RCTM_User, time=0, value=0}
           {interpMode=RCIM_Cubic, tangentMode=RCTM_User, time=0.5, value=1}
           {tangentMode=RCTM_User, time=1, value=0}
          ]}}}
       Scale Curve=&MaskFloatBySpawnGroup(Masked Float=0.3, Passthrough Float=0.1)
      )
      B=&Lerp_Float(
       A=0
       B=1
       Alpha=&RemapRange(
        Input Value To Remap=&Abs_Float(Float:NiagaraFloat=&DotProduct(
          Vector A=&NormalizeVector(
           Vector To Normalize=&ReturnCameraVector(Camera Vectors=::NewEnumerator1, Coordinate Space=::World)
           Safe Normalize=true
           Safe Fallback Threshold=0
           Safe Fallback Vector=<0 0 0.001>
          )
          Vector B=&NormalizeVector(Vector To Normalize=$Engine.Owner.SystemXAxis, Safe Normalize=true, Safe Fallback Threshold=0, Safe Fallback Vector=<0 0 0.001>)
         ))
        Input Min=0
        Input Max=1
        Output Min=-1
        Output Max=1
       )
      )
     )
    )) {isSetParameters=true}
  ]
  rendererClasses [&NiagaraSpriteRendererProperties]
  renderers [
   {rendererClass=&NiagaraSpriteRendererProperties}
   {rendererClass=&NiagaraSpriteRendererProperties, rendererIndex=1}
  ]
 }
}
summary {
 systemName=NS_VelvetDeskCam_Smiley
 emitters [
  {emitterName=Main, rendererClasses [&NiagaraMeshRendererProperties]}
  {emitterName=Smoke, rendererClasses [&NiagaraSpriteRendererProperties]}
 ]
 userVariables [
  User.bIsFrontEnd {type=NiagaraBool}
  User.bIsFrontEndPreview {type=NiagaraBool}
  User.BoxSize {type=Vector3f, description="", defaultValue {struct=Vector3f, value=<10 500 100>}}
  User.SmileSize {type=NiagaraFloat, defaultValue=1, description=""}
 ]
}
